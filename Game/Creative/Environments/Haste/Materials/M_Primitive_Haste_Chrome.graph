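# Chrome material for Haste Creative environment primitives: tinted, metallic by default, and usable on instanced static meshes.
# A world-aligned noise texture drives roughness between low and high values.
# World-space normals layer two triplanar textures over either the vertex normal or a UV normal map, with static switches selecting triplanar or optional detail blending.
# Material attributes pass through MF_GameWideShared before output.
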Material /Game/Creative/Environments/Haste/Materials/M_Primitive_Haste_Chrome
graph {
 properties {
  bTangentSpaceNormal=false
  bUseMaterialAttributes=true
  bUsedWithInstancedStaticMeshes=true
  pixelDepthOffsetMode=PDOM_Legacy
  usedShadingModels=Default Lit
 }
 nodes [
  n1 = VertexNormalWS()
  n2 = TextureSampleParameter2D() {
   group=Normal
   parameterName=Normal Map
   samplerType=SAMPLERTYPE_Normal
   sortPriority=10
   texture=/Engine/EngineMaterials/BaseFlattenNormalMap
  }
  n3 = Transform(None=@n2.RGB)
  n4 = ScalarParameter() {defaultValue=1024, group=Normal, parameterName=Triplanar Normal Texture Size, sortPriority=101}
  n5 = TextureObjectParameter() {
   group=Normal
   parameterName=Triplanar Normal
   samplerType=SAMPLERTYPE_Normal
   sortPriority=100
   texture=/Game/Global/Textures/Default/Noise/T_Noise_Normal_01
  }
  n6 = TextureObjectParameter() {
   group=Normal
   parameterName=Triplanar Detail Normal Texture
   samplerType=SAMPLERTYPE_Normal
   sortPriority=200
   texture=/Game/Weapons/WeaponSkins/Wraps/Textures/IceBreaker/T_IceNormal
  }
  n7 = LinearInterpolate(
   A=LinearInterpolate(
    A=StaticSwitchParameter(True=@n1, False=@n3) {group=Normal, parameterName=Use Triplanar Normal, sortPriority=0}
    B=MaterialFunctionCall!(Negative World Position Divisor=@n4, World Position, Texture=@n5, Z lerp Alpha, WorldNormal, X lerp alpha) {materialFunction=/Engine/Functions/Engine_MaterialFunctions02/WorldAlignedNormals_HighQuality}.World Space Normals XYZ
    Alpha=ScalarParameter() {defaultValue=0.4, group=Normal, parameterName=Triplanar Normal Strength, sortPriority=102}
   )
   B=MaterialFunctionCall(
    Negative World Position Divisor=ScalarParameter() {defaultValue=128, group=Normal, parameterName=Triplanar Detail Normal Texture Size, sortPriority=201}
    Texture=@n6
   ) {materialFunction=&WorldAlignedNormals_HighQuality}.World Space Normals XYZ
   Alpha=ScalarParameter() {defaultValue=0.25, group=Normal, parameterName=Triplanar Detail Normal Strength, sortPriority=202}
  )
  n8 = StaticSwitchParameter(
   True=LinearInterpolate(A=@n3, B=@n7, Alpha=ScalarParameter() {group=Normal, parameterName=Detail Normal strength, sortPriority=11})
   False=@n3
  ) {group=Normal, parameterName=Add Normal Detail, sortPriority=1}
  n9 = ScalarParameter() {defaultValue=0.2, group=Roughness, parameterName=Roughness Low, sortPriority=3}
  n10 = ScalarParameter() {defaultValue=0.5, group=Roughness, parameterName=Roughness Max, sortPriority=2}
  n11 = TextureObjectParameter() {group=Roughness, parameterName=Roughness Texture, sortPriority=0, texture=&AtmosphericCloudNoise01}
  n12 = ComponentMask(None=MaterialFunctionCall(TextureObject=@n11, TextureSize=ScalarParameter() {defaultValue=1024, group=Roughness, parameterName=Roughness Texture Size, sortPriority=1}) {materialFunction=/Game/Weapons/WeaponSkins/Wraps/MaterialFunctions/WorldAlignedTexture}.XYZ Texture) {r=true}
  MaterialFunctionCall!(In=MakeMaterialAttributes!(
    BaseColor=VectorParameter() {group=Color, parameterName=Color, sortPriority=0, defaultValue=<rgba 0.3911 0.6235 0.7011 1>}.RGB
    Metallic=ScalarParameter() {defaultValue=1, group=Color, parameterName=Metallic, sortPriority=1}
    Specular
    Roughness=LinearInterpolate(A=@n9, B=@n10, Alpha=@n12)
    Anisotropy
    EmissiveColor
    Opacity
    OpacityMask
    Normal=StaticSwitchParameter(True=@n7, False=@n8) {group=Normal, parameterName=Use Triplanar Normal, sortPriority=0}
    Tangent
    WorldPositionOffset
    SubsurfaceColor
    ClearCoat
    ClearCoatRoughness
    AmbientOcclusion
    Refraction
    CustomizedUVs_0
    CustomizedUVs_1
    CustomizedUVs_2
    CustomizedUVs_3
    CustomizedUVs_4
    CustomizedUVs_5
    CustomizedUVs_6
    CustomizedUVs_7
    PixelDepthOffset
    ShadingModel
    Displacement
   )) {materialFunction=/Game/Athena/WinterArt/Materials/MF_GameWideShared}
 ]
}
