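# Opaque multicolor material for Haste Creative primitives, blending near and far looks, apparently by a camera depth fade.
# Base color tints world-XY-projected splatter textures, shifting UVs by a normal-to-view dot product for an iridescent effect.
# World-aligned textures drive roughness and layered world-space normals, with metallic also set per range.
# Attributes pass through MF_GameWideShared; the CheapContrast node's output is unconnected.
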Material /Game/Creative/Environments/Haste/Materials/M_Primitive_Haste_Multicolor
graph {
 properties {
  bTangentSpaceNormal=false
  bUseMaterialAttributes=true
  bUsedWithInstancedStaticMeshes=true
  pixelDepthOffsetMode=PDOM_Legacy
  usedShadingModels=Default Lit
 }
 nodes [
  n1 = WorldPosition()
  n2 = Constant() {r=1000}
  n3 = ScalarParameter() {defaultValue=7.407, group=Color, parameterName=Color Noise Tiling Near, sortPriority=3}
  n4 = Multiply(
   A=DotProduct(A=NamedRerouteUsage(), B=CameraVectorWS())
   B=ScalarParameter() {defaultValue=0.1, group=Color, parameterName=Normal Iredescence, sortPriority=2}
  )
  n5 = Add(A=Divide(A=ComponentMask(None=@n1.XYZ) {g=true, r=true}, B=Multiply(A=@n2, B=@n3)), B=@n4)
  n6 = TextureObjectParameter() {group=Color, parameterName=Color Splatter Near, texture=&/T_Hase_Splatter_C_M}
  n7 = WorldPosition()
  n8 = Constant() {r=1000}
  n9 = ScalarParameter() {defaultValue=7.407, group=Color, parameterName=Color Noise Tiling Far, sortPriority=3}
  n10 = Add(A=Divide(A=ComponentMask(None=@n7.XYZ) {g=true, r=true}, B=Multiply(A=@n8, B=@n9)), B=@n4)
  n11 = TextureObjectParameter() {group=Color, parameterName=Color Splatter Far, texture=&/T_Hase_Splatter_C2_M}
  n12 = LinearInterpolate(
   A=Desaturation(
    None=TextureSample(UVs=@n5, Tex=@n6) {texture=&AtmosphericCloudNoise01}.RGB
    Fraction=ScalarParameter() {group=Color, parameterName=Desaturation Near}
   )
   B=Desaturation(
    None=TextureSample(UVs=@n10, Tex=@n11) {texture=&AtmosphericCloudNoise01}.RGB
    Fraction=ScalarParameter() {group=Color, parameterName=Desaturation Far}
   )
   Alpha=NamedRerouteUsage()
  )
  MaterialFunctionCall(In=@n12, Contrast=ScalarParameter() {group=Color, parameterName="Contrast (S)"}) {materialFunction=&CheapContrast}
  MaterialFunctionCall!(In=MakeMaterialAttributes!(
    BaseColor=NamedRerouteUsage()
    Metallic=NamedRerouteUsage()
    Specular
    Roughness=NamedRerouteUsage()
    Anisotropy
    EmissiveColor
    Opacity
    OpacityMask
    Normal=NamedRerouteUsage()
    Tangent
    WorldPositionOffset
    SubsurfaceColor
    ClearCoat
    ClearCoatRoughness
    AmbientOcclusion
    Refraction
    CustomizedUVs_0
    CustomizedUVs_1
    CustomizedUVs_2
    CustomizedUVs_3
    CustomizedUVs_4
    CustomizedUVs_5
    CustomizedUVs_6
    CustomizedUVs_7
    PixelDepthOffset
    ShadingModel
    Displacement
   )) {materialFunction=/Game/Athena/WinterArt/Materials/MF_GameWideShared}
  n13 = TextureObjectParameter() {group=Roughness, parameterName=Roughness Texture, sortPriority=0, texture=&AtmosphericCloudNoise01}
  n14 = ScalarParameter() {defaultValue=1024, group=Roughness, parameterName=Roughness Texture Size, sortPriority=1}
  n15 = LinearInterpolate(
   A=ScalarParameter() {defaultValue=0.2, group=Roughness, parameterName=Roughness Low, sortPriority=3}
   B=ScalarParameter() {defaultValue=0.5, group=Roughness, parameterName=Roughness Max, sortPriority=2}
   Alpha=ComponentMask(None=MaterialFunctionCall(TextureObject=@n13, TextureSize=@n14) {materialFunction=/Game/Weapons/WeaponSkins/Wraps/MaterialFunctions/WorldAlignedTexture}.XYZ Texture) {r=true}
  )
  NamedRerouteDeclaration(None=LinearInterpolate(A=@n15, B=ScalarParameter() {defaultValue=0.5, group=Roughness, parameterName=Roughness Far, sortPriority=3}, Alpha=NamedRerouteUsage())) {name=Roughness, nodeColor=<rgba 1 0.1875 0 1>}
  n16 = VectorParameter() {group=Color, parameterName=Color Near, sortPriority=0, defaultValue=<rgba 1 1 1 1>}
  n17 = VectorParameter() {group=Color, parameterName=Color Far, sortPriority=0, defaultValue=<rgba 1 1 1 1>}
  n18 = NamedRerouteUsage()
  NamedRerouteDeclaration(None=Multiply(A=@n12, B=LinearInterpolate(A=@n16.RGB, B=@n17.RGB, Alpha=@n18))) {name=Base Color, nodeColor=<rgba 0 1 0.0391 1>}
  NamedRerouteDeclaration(None=LinearInterpolate(
    A=ScalarParameter() {defaultValue=1, group=Color, parameterName=Metallic Near, sortPriority=1}
    B=ScalarParameter() {defaultValue=1, group=Color, parameterName=Metallic Far, sortPriority=1}
    Alpha=NamedRerouteUsage()
   )) {name=Metallic, nodeColor=<rgba 0.4531 0 1 1>}
  n19 = TextureObjectParameter() {
   group=Normal
   parameterName=Near Normal Tex
   samplerType=SAMPLERTYPE_Normal
   sortPriority=100
   texture=/Game/Global/Textures/Default/Noise/T_Noise_Normal_01
  }
  n20 = LinearInterpolate(
   A=VertexNormalWS()
   B=MaterialFunctionCall!(
    Negative World Position Divisor=ScalarParameter() {defaultValue=1024, group=Normal, parameterName=Near Normal Tiling, sortPriority=101}
    World Position
    Texture=@n19
    Z lerp Alpha
    WorldNormal
    X lerp alpha
   ) {materialFunction=/Engine/Functions/Engine_MaterialFunctions02/WorldAlignedNormals_HighQuality}.World Space Normals XYZ
   Alpha=ScalarParameter() {defaultValue=0.4, group=Normal, parameterName=Near Normal Strength, sortPriority=102}
  )
  n21 = TextureObjectParameter() {
   group=Normal
   parameterName=Near Normal Detail Tex
   samplerType=SAMPLERTYPE_Normal
   sortPriority=110
   texture=/Game/Weapons/WeaponSkins/Wraps/Textures/IceBreaker/T_IceNormal
  }
  n22 = MaterialFunctionCall(
   Negative World Position Divisor=ScalarParameter() {defaultValue=128, group=Normal, parameterName=Near Normal Detail Tiling, sortPriority=111}
   Texture=@n21
  ) {materialFunction=&WorldAlignedNormals_HighQuality}
  n23 = ScalarParameter() {defaultValue=0.6, group=Normal, parameterName=Near Normal Detail Stregnth, sortPriority=112}
  n24 = TextureObjectParameter() {
   group=Normal
   parameterName=Far Normal Tex
   samplerType=SAMPLERTYPE_Normal
   sortPriority=200
   texture=&T_Noise_Normal_01
  }
  n25 = LinearInterpolate(
   A=VertexNormalWS()
   B=MaterialFunctionCall(
    Negative World Position Divisor=ScalarParameter() {defaultValue=1024, group=Normal, parameterName=Far Normal Tiling, sortPriority=201}
    Texture=@n24
   ) {materialFunction=&WorldAlignedNormals_HighQuality}.World Space Normals XYZ
   Alpha=ScalarParameter() {defaultValue=0.4, group=Normal, parameterName=Far Normal Strength, sortPriority=202}
  )
  n26 = TextureObjectParameter() {
   group=Normal
   parameterName=Far Detail Normal Tex
   samplerType=SAMPLERTYPE_Normal
   sortPriority=210
   texture=&T_IceNormal
  }
  n27 = MaterialFunctionCall(
   Negative World Position Divisor=ScalarParameter() {defaultValue=128, group=Normal, parameterName=Far Normal Detail Tiling, sortPriority=211}
   Texture=@n26
  ) {materialFunction=&WorldAlignedNormals_HighQuality}
  n28 = ScalarParameter() {defaultValue=0.4, group=Normal, parameterName=Far Normal Detail Strength, sortPriority=212}
  NamedRerouteDeclaration(None=LinearInterpolate(
    A=LinearInterpolate(A=@n20, B=@n22.World Space Normals XYZ, Alpha=@n23)
    B=LinearInterpolate(A=@n25, B=@n27.World Space Normals XYZ, Alpha=@n28)
    Alpha=NamedRerouteUsage()
   ) {constA=1, constB=0.2}) {name=Normals, nodeColor=<rgba 0 0.3906 1 1>}
  NamedRerouteDeclaration(None=MaterialFunctionCall(
    Fade Length=ScalarParameter() {defaultValue=5000, parameterName=Camera Distance Length}
    Fade Offset=ScalarParameter() {defaultValue=2000, parameterName=Camera Distance Offset}
   ) {materialFunction=&CameraDepthFade}.Result) {name=Distance Fade, nodeColor=<rgba 1 0 0.1556 1>}
  Reroute(None=@n15)
  Reroute()
 ]
}
